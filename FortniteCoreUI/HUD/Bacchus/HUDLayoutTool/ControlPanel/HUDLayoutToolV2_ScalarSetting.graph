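# Settings list entry widget for a numeric setting in the HUD Layout Tool V2 control panel, built on the native HUDLayoutToolV2SettingsListEntry_Scalar class.
# It shows name and value text above an arrow-button stepper around an analog slider, which has a transparent bar drawn over a material progress bar.
# Value changes drive that progress bar, while default-value changes set a scalar parameter on a collapsed-by-default marker image's dynamic material.
# The slider is the primary gamepad focus target.
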WidgetBlueprint /FortniteCoreUI/HUD/Bacchus/HUDLayoutTool/ControlPanel/HUDLayoutToolV2_ScalarSetting
graphs [
 EventGraph {nodes [
   n1 = Event OnValueChanged!() -> (OutputDelegate, then, Value:float) {is=title K2Node_Event}
   > SetProgress!(
    execute
    self:Fort Simple Material Progress Bar Object Reference=Get Progress() -> (Progress:Simple Material Progress Bar Object Reference)
    Progress:float=@n1.Value
   ) -> (then)
   n2 = Event OnDefaultValueChanged!() -> (OutputDelegate, then, DefaultValue:float) {is=title K2Node_Event}
   > n3 = GetDynamicMaterial(self=Get DefaultValueLayer() -> (DefaultValueLayer:Image Object Reference))
   > SetScalarParameterValue(self=@n3, ParameterName=DefaultValue, Value=@n2.DefaultValue)
  ]}
 GetPrimaryGamepadFocusWidget {nodes [
   GetPrimaryGamepadFocusWidget() {is=title K2Node_FunctionEntry}
   > Return Node!(execute, ReturnValue:Widget Object Reference=Get Slider_SettingValue() -> (Slider_SettingValue:Analog Slider Object Reference))
  ]}
]
variables []
widgets {
 info {
  inheritedWidgetCount=7
  parentClass=/Script/FortniteUI.HUDLayoutToolV2SettingsListEntry_Scalar
  rootWidgetClass=&VerticalBox
  widgetCount=15
 }
 widgets [
  VerticalBox_1 {widgetClassPath=&VerticalBox, widget=VerticalBox_1}
  HorizontalBox_1 {
   widgetClassPath=&HorizontalBox
   parent=VerticalBox_1
   widget=HorizontalBox_1
   slot {class=VerticalBoxSlot, properties {padding=<ltrb 0 0 90 0>}}
  }
  SizeBox_0 {
   widgetClassPath=&SizeBox
   parent=HorizontalBox_1
   widget=SizeBox_0
   properties {maxDesiredWidth=450, minDesiredWidth=350, widthOverride=325}
   slot {class=HorizontalBoxSlot, properties {verticalAlignment=VAlign_Bottom, size {sizeRule=Fill, value=1}}}
  }
  Text_SettingName {
   widgetClassPath=&CommonTextBlock
   bInherited=true
   parent=SizeBox_0
   widget=Text_SettingName
   properties {
    scrollStyle=&TextScrollStyle-Base_C
    text="{Setting Name}"
    colorAndOpacity {colorUseRule=UseColor_Specified, specifiedColor=<rgba 0.4125 0.4125 0.4125 1>}
    font {
     bForceMonospaced=false
     bMaterialIsStencil=false
     fontMaterial=null
     fontObject=&BurbankSmall
     letterSpacing=0
     monospacedWidth=1
     size=22
     skewAmount=0
     typefaceFontName=Bold
     outlineSettings {
      bApplyOutlineToDropShadows=false
      bMiteredCorners=false
      bSeparateFillAlpha=false
      outlineMaterial=null
      outlineSize=0
      outlineColor=<rgba 0 0 0 1>
     }
    }
    shadowOffset=<0 0>
   }
   slot {class=SizeBoxSlot}
  }
  Text_SettingValue {
   widgetClassPath=&CommonTextBlock
   bInherited=true
   parent=HorizontalBox_1
   widget=Text_SettingValue
   properties {
    justification=Right
    scrollStyle=&TextScrollStyle-Base_C
    text="{Value}"
    textTransformPolicy=ToUpper
    visibility=HitTestInvisible
    font {
     bForceMonospaced=false
     bMaterialIsStencil=false
     fontMaterial=null
     fontObject=&BurbankSmall
     letterSpacing=0
     monospacedWidth=1
     size=24
     skewAmount=0
     typefaceFontName=Bold
     outlineSettings {
      bApplyOutlineToDropShadows=false
      bMiteredCorners=false
      bSeparateFillAlpha=false
      outlineMaterial=null
      outlineSize=0
      outlineColor=<rgba 0 0 0 1>
     }
    }
    renderTransform {angle=0, scale=<1 1>, shear=<0 0>, translation=<0 2>}
    shadowOffset=<0 0>
   }
   slot {class=HorizontalBoxSlot, properties {verticalAlignment=VAlign_Bottom, padding=<ltrb 40 0 0 0>}}
  }
  HorizontalBox_0 {
   widgetClassPath=&HorizontalBox
   parent=VerticalBox_1
   widget=HorizontalBox_0
   slot {class=VerticalBoxSlot, properties {padding=<ltrb 0 6 0 0>}}
  }
  Button_DecreaseValue {
   widgetClassPath=/Game/Athena/HUD/Bacchus/HUDLayoutTool/HUDLayoutToolV2_Button
   bInherited=true
   parent=HorizontalBox_0
   widget=Button_DecreaseValue
   properties {
    bOverride_HeightOverride=true
    bOverride_WidthOverride=true
    buttonHeightOverride=80
    buttonWidthOverride=70
    text=""
    backgroundBrush {
     bIsDynamicallyLoaded=false
     drawAs=Border
     imageType=NoImage
     mirroring=NoMirror
     resourceName=None
     resourceObject=null
     tiling=NoTile
     imageSize=<32 32>
     margin=<ltrb 0 0 0 0>
     outlineSettings {
      bUseBrushTransparency=false
      roundingType=HalfHeightRadius
      width=0
      color {colorUseRule=UseColor_Specified, specifiedColor=<rgba 0 0 0 0>}
      cornerRadii=<0 0 0 0>
     }
     tintColor {colorUseRule=UseColor_Specified, specifiedColor=<rgba 1 1 1 1>}
     uVRegion {bIsValid=false, max=<0 0>, min=<0 0>}
    }
    iconBrush {
     bIsDynamicallyLoaded=false
     drawAs=Image
     imageType=NoImage
     mirroring=NoMirror
     resourceName=None
     resourceObject=/Game/Athena/HUD/Bacchus/HUDLayoutTool/HLT_Icons/T-HLT-Arrow_Sprite
     tiling=NoTile
     imageSize=<50 38>
     margin=<ltrb 0 0 0 0>
     outlineSettings {
      bUseBrushTransparency=false
      roundingType=HalfHeightRadius
      width=0
      color {colorUseRule=UseColor_Specified, specifiedColor=<rgba 0 0 0 0>}
      cornerRadii=<0 0 0 0>
     }
     tintColor {colorUseRule=UseColor_Specified, specifiedColor=<rgba 1 1 1 1>}
     uVRegion {bIsValid=false, max=<0 0>, min=<0 0>}
    }
    internalPadding=<ltrb 0 0 0 0>
    pressedButtonStyle {
     iconAngle=0
     backgroundColor=<rgba 1 1 1 1>
     iconColor=<rgba 0 0.117 1 1>
     textColor {colorUseRule=UseColor_Specified, specifiedColor=<rgba 1 1 1 1>}
    }
    renderTransform {angle=-90, scale=<1 1>, shear=<0 0>, translation=<5 0>}
   }
   slot {class=HorizontalBoxSlot, properties {size {sizeRule=Automatic, value=0.1}}}
  }
  SizeBox_1 {
   widgetClassPath=&SizeBox
   parent=HorizontalBox_0
   widget=SizeBox_1
   properties {bOverride_MinDesiredHeight=true, minDesiredHeight=50}
   slot {class=HorizontalBoxSlot, properties {padding=<ltrb 20 0 20 0>, size {sizeRule=Fill, value=1}}}
  }
  Panel_Value {widgetClassPath=&Overlay, bInherited=true, parent=SizeBox_1, widget=Panel_Value, slot {class=SizeBoxSlot}}
  SizeBox_3 {
   widgetClassPath=&SizeBox
   parent=Panel_Value
   widget=SizeBox_3
   properties {bOverride_HeightOverride=true, heightOverride=6}
   slot {class=OverlaySlot, properties {horizontalAlignment=HAlign_Fill, verticalAlignment=VAlign_Center}}
  }
  Progress {
   widgetClassPath=/Game/Athena/HUD/SimpleMaterialProgressBar
   bIsVariable=true
   parent=SizeBox_3
   widget=Progress
   properties {
    bOverrideDefaultColorA=true
    bOverrideDefaultColorB=true
    bOverrideDefaultColorBackground=true
    bUseStroke=false
    designTime_Progress=0.25
    visibility=HitTestInvisible
    cachedColorA=<rgba 0.3095 0.3095 0.3095 1>
    cachedColorB=<rgba 0.3095 0.3095 0.3095 1>
    cachedColorBackground=<rgba 0 0 0 1>
   }
   slot {class=SizeBoxSlot, properties {padding=<ltrb 30 0 30 0>}}
  }
  DefaultValueLayer {
   widgetClassPath=&Image
   bIsVariable=true
   parent=Panel_Value
   widget=DefaultValueLayer
   properties {
    visibility=Collapsed
    brush {
     bIsDynamicallyLoaded=false
     drawAs=Image
     imageType=NoImage
     mirroring=NoMirror
     resourceName=None
     resourceObject=/Game/Athena/Settings/Art/SettingsDefaultSlider_M
     tiling=NoTile
     imageSize=<32 32>
     margin=<ltrb 0 0 0 0>
     outlineSettings {
      bUseBrushTransparency=false
      roundingType=HalfHeightRadius
      width=0
      color {colorUseRule=UseColor_Specified, specifiedColor=<rgba 0 0 0 0>}
      cornerRadii=<0 0 0 0>
     }
     tintColor {colorUseRule=UseColor_Specified, specifiedColor=<rgba 1 1 1 1>}
     uVRegion {bIsValid=false, max=<0 0>, min=<0 0>}
    }
   }
   slot {class=OverlaySlot, properties {horizontalAlignment=HAlign_Fill, verticalAlignment=VAlign_Fill, padding=<ltrb 31 24 31 24>}}
  }
  Slider_SettingValue {
   widgetClassPath=/Script/CommonUI.AnalogSlider
   bInherited=true
   parent=Panel_Value
   widget=Slider_SettingValue
   properties {
    value=0.25
    sliderBarColor=<rgba 0 0 0 0>
    widgetStyle {
     barThickness=6
     disabledBarImage {
      bIsDynamicallyLoaded=false
      drawAs=Image
      imageType=NoImage
      mirroring=NoMirror
      resourceName=None
      resourceObject=null
      tiling=NoTile
      imageSize=<0 0>
      margin=<ltrb 0 0 0 0>
      outlineSettings {
       bUseBrushTransparency=false
       roundingType=HalfHeightRadius
       width=0
       color {colorUseRule=UseColor_Specified, specifiedColor=<rgba 0 0 0 0>}
       cornerRadii=<0 0 0 0>
      }
      tintColor {colorUseRule=UseColor_Specified, specifiedColor=<rgba 0.3095 0.3095 0.3095 1>}
      uVRegion {bIsValid=false, max=<0 0>, min=<0 0>}
     }
     disabledThumbImage {
      bIsDynamicallyLoaded=false
      drawAs=Image
      imageType=FullColor
      mirroring=NoMirror
      resourceName=../../../Engine/Content/Slate/Common/Button_Disabled.png
      resourceObject=/FortniteCoreUI/Athena/HUD/Bacchus/HUDLayoutTool/HLT_Icons/MI_HUDLayoutToolV2_Handle
      tiling=NoTile
      imageSize=<60 54>
      margin=<ltrb 0 0 0 0>
      outlineSettings {
       bUseBrushTransparency=false
       roundingType=HalfHeightRadius
       width=0
       color {colorUseRule=UseColor_Specified, specifiedColor=<rgba 0 0 0 0>}
       cornerRadii=<0 0 0 0>
      }
      tintColor {colorUseRule=UseColor_Specified, specifiedColor=<rgba 0.5029 0.5029 0.5029 1>}
      uVRegion {bIsValid=false, max=<0 0>, min=<0 0>}
     }
     hoveredBarImage {
      bIsDynamicallyLoaded=false
      drawAs=Image
      imageType=NoImage
      mirroring=NoMirror
      resourceName=None
      resourceObject=null
      tiling=NoTile
      imageSize=<0 0>
      margin=<ltrb 0 0 0 0>
      outlineSettings {
       bUseBrushTransparency=false
       roundingType=HalfHeightRadius
       width=0
       color {colorUseRule=UseColor_Specified, specifiedColor=<rgba 0 0 0 0>}
       cornerRadii=<0 0 0 0>
      }
      tintColor {colorUseRule=UseColor_Specified, specifiedColor=<rgba 0.3095 0.3095 0.3095 1>}
      uVRegion {bIsValid=false, max=<0 0>, min=<0 0>}
     }
     hoveredThumbImage {
      bIsDynamicallyLoaded=false
      drawAs=Image
      imageType=FullColor
      mirroring=NoMirror
      resourceName=../../../Engine/Content/Slate/Common/Button.png
      resourceObject=&MI_HUDLayoutToolV2_Handle
      tiling=NoTile
      imageSize=<60 54>
      margin=<ltrb 0 0 0 0>
      outlineSettings {
       bUseBrushTransparency=false
       roundingType=HalfHeightRadius
       width=0
       color {colorUseRule=UseColor_Specified, specifiedColor=<rgba 0 0 0 0>}
       cornerRadii=<0 0 0 0>
      }
      tintColor {colorUseRule=UseColor_Specified, specifiedColor=<rgba 1 1 1 1>}
      uVRegion {bIsValid=false, max=<0 0>, min=<0 0>}
     }
     normalBarImage {
      bIsDynamicallyLoaded=false
      drawAs=Image
      imageType=NoImage
      mirroring=NoMirror
      resourceName=None
      resourceObject=null
      tiling=NoTile
      imageSize=<0 0>
      margin=<ltrb 0 0 0 0>
      outlineSettings {
       bUseBrushTransparency=false
       roundingType=HalfHeightRadius
       width=0
       color {colorUseRule=UseColor_Specified, specifiedColor=<rgba 0 0 0 0>}
       cornerRadii=<0 0 0 0>
      }
      tintColor {colorUseRule=UseColor_Specified, specifiedColor=<rgba 0.3095 0.3095 0.3095 1>}
      uVRegion {bIsValid=false, max=<0 0>, min=<0 0>}
     }
     normalThumbImage {
      bIsDynamicallyLoaded=false
      drawAs=Image
      imageType=FullColor
      mirroring=NoMirror
      resourceName=../../../Engine/Content/Slate/Common/Button.png
      resourceObject=&MI_HUDLayoutToolV2_Handle
      tiling=NoTile
      imageSize=<60 54>
      margin=<ltrb 0 0 0 0>
      outlineSettings {
       bUseBrushTransparency=false
       roundingType=HalfHeightRadius
       width=0
       color {colorUseRule=UseColor_Specified, specifiedColor=<rgba 0 0 0 0>}
       cornerRadii=<0 0 0 0>
      }
      tintColor {colorUseRule=UseColor_Specified, specifiedColor=<rgba 1 1 1 1>}
      uVRegion {bIsValid=false, max=<0 0>, min=<0 0>}
     }
    }
   }
   slot {class=OverlaySlot, properties {horizontalAlignment=HAlign_Fill, verticalAlignment=VAlign_Fill}}
  }
  Button_IncreaseValue {
   widgetClassPath=&HUDLayoutToolV2_Button
   bInherited=true
   parent=HorizontalBox_0
   widget=Button_IncreaseValue
   properties {
    bOverride_HeightOverride=true
    bOverride_WidthOverride=true
    buttonHeightOverride=80
    buttonWidthOverride=70
    text=""
    backgroundBrush {
     bIsDynamicallyLoaded=false
     drawAs=Border
     imageType=NoImage
     mirroring=NoMirror
     resourceName=None
     resourceObject=null
     tiling=NoTile
     imageSize=<32 32>
     margin=<ltrb 0 0 0 0>
     outlineSettings {
      bUseBrushTransparency=false
      roundingType=HalfHeightRadius
      width=0
      color {colorUseRule=UseColor_Specified, specifiedColor=<rgba 0 0 0 0>}
      cornerRadii=<0 0 0 0>
     }
     tintColor {colorUseRule=UseColor_Specified, specifiedColor=<rgba 1 1 1 1>}
     uVRegion {bIsValid=false, max=<0 0>, min=<0 0>}
    }
    iconBrush {
     bIsDynamicallyLoaded=false
     drawAs=Image
     imageType=NoImage
     mirroring=NoMirror
     resourceName=None
     resourceObject=&T-HLT-Arrow_Sprite
     tiling=NoTile
     imageSize=<50 38>
     margin=<ltrb 0 0 0 0>
     outlineSettings {
      bUseBrushTransparency=false
      roundingType=HalfHeightRadius
      width=0
      color {colorUseRule=UseColor_Specified, specifiedColor=<rgba 0 0 0 0>}
      cornerRadii=<0 0 0 0>
     }
     tintColor {colorUseRule=UseColor_Specified, specifiedColor=<rgba 1 1 1 1>}
     uVRegion {bIsValid=false, max=<0 0>, min=<0 0>}
    }
    internalPadding=<ltrb 0 0 0 0>
    pressedButtonStyle {
     iconAngle=0
     backgroundColor=<rgba 1 1 1 1>
     iconColor=<rgba 0 0.117 1 1>
     textColor {colorUseRule=UseColor_Specified, specifiedColor=<rgba 1 1 1 1>}
    }
    renderTransform {angle=90, scale=<1 1>, shear=<0 0>, translation=<-5 0>}
   }
   slot {class=HorizontalBoxSlot, properties {size {sizeRule=Automatic, value=0.1}}}
  }
  Background {widgetClassPath=&UserWidget, bInherited=true, widget=null}
 ]
}
