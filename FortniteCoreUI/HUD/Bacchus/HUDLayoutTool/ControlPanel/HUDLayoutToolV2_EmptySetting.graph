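# Settings list entry widget for the mobile HUD Layout Tool V2 control panel that serves as a placeholder row reading "No Editor Found".
# It derives from FortSettingsListEntry_Setting and fills the inherited Background slot with a responsive panel of two minimum-height size boxes.
# One box holds the inherited, blue, uppercase setting-name text, and the other holds the static centered placeholder message.
# The event graph is empty, so any behavior comes from the parent class.
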WidgetBlueprint /FortniteCoreUI/HUD/Bacchus/HUDLayoutTool/ControlPanel/HUDLayoutToolV2_EmptySetting
graphs [EventGraph {links [], nodes []}]
variables []
widgets {
 info {
  inheritedWidgetCount=2
  namedSlotCount=1
  parentClass=/Script/FortniteUI.FortSettingsListEntry_Setting
  rootWidgetClass=/FortniteCoreUI/Settings/SettingEntryBackground.SettingEntryBackground_C
  widgetCount=6
 }
 widgets [
  Background {widgetClassPath=/FortniteCoreUI/Settings/SettingEntryBackground, bInherited=true, widget=Background}
  FortResponsivePanel_0 {widgetClassPath=/Script/FortniteUI.FortResponsivePanel, namedSlotHost=Background, widget=FortResponsivePanel_0}
  SizeBox_0 {
   widgetClassPath=&SizeBox
   parent=FortResponsivePanel_0
   widget=SizeBox_0
   properties {bOverride_MinDesiredHeight=true, minDesiredHeight=50}
   slot {class=FortResponsivePanelSlot}
  }
  Text_SettingName {
   widgetClassPath=&CommonTextBlock
   bInherited=true
   parent=SizeBox_0
   widget=Text_SettingName
   properties {
    scrollStyle=&TextScrollStyle-Base_C
    text="{Setting Name}"
    textTransformPolicy=ToUpper
    colorAndOpacity {colorUseRule=UseColor_Specified, specifiedColor=<rgba 0.1946 0.5972 0.9473 1>}
    font {
     bForceMonospaced=false
     bMaterialIsStencil=false
     fontMaterial=null
     fontObject=&BurbankSmall
     letterSpacing=0
     monospacedWidth=1
     size=18
     skewAmount=0
     typefaceFontName=Black
     outlineSettings {
      bApplyOutlineToDropShadows=false
      bMiteredCorners=false
      bSeparateFillAlpha=false
      outlineMaterial=null
      outlineSize=0
      outlineColor=<rgba 0 0 0 1>
     }
    }
    margin=<ltrb 10 0 0 0>
    shadowOffset=<0 0>
   }
   slot {class=SizeBoxSlot, properties {verticalAlignment=VAlign_Center}}
  }
  SizeBox {
   widgetClassPath=&SizeBox
   parent=FortResponsivePanel_0
   widget=SizeBox
   properties {bOverride_MinDesiredHeight=true, minDesiredHeight=50}
   slot {class=FortResponsivePanelSlot}
  }
  Text_SettingName_1 {
   widgetClassPath=&CommonTextBlock
   parent=SizeBox
   widget=Text_SettingName_1
   properties {
    justification=Center
    scrollStyle=&TextScrollStyle-Base_C
    text=No Editor Found
    textTransformPolicy=ToUpper
    font {
     bForceMonospaced=false
     bMaterialIsStencil=false
     fontMaterial=null
     fontObject=&BurbankSmall
     letterSpacing=0
     monospacedWidth=1
     size=18
     skewAmount=0
     typefaceFontName=Black
     outlineSettings {
      bApplyOutlineToDropShadows=false
      bMiteredCorners=false
      bSeparateFillAlpha=false
      outlineMaterial=null
      outlineSize=0
      outlineColor=<rgba 0 0 0 1>
     }
    }
    margin=<ltrb 10 0 0 0>
    shadowOffset=<0 0>
   }
   slot {class=SizeBoxSlot, properties {verticalAlignment=VAlign_Center}}
  }
 ]
}
